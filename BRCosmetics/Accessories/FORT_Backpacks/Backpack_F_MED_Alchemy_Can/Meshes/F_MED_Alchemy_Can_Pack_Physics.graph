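# Physics asset for the Alchemy Can backpack cosmetic mesh, defining simple collision bodies for its two dyn_ bones.
# A thin square box represents dyn_liquid and a capsule represents dyn_pack.
# One constraint joins dyn_liquid to dyn_pack with twist locked and matching narrow limit angles on both swing axes.
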PhysicsAsset /BRCosmetics/Accessories/FORT_Backpacks/Backpack_F_MED_Alchemy_Can/Meshes/F_MED_Alchemy_Can_Pack_Physics
bodies {
 dyn_liquid [{shapeType=Box, extentX=11.01, extentY=11.01, extentZ=1.01, center=<1.1768 0 0>, rotation=<rot 90 0 0>}]
 dyn_pack [{shapeType=Capsule, length=3.9548, radius=4.7198, center=<-0.0146 -3.9209 -7.577>, rotation=<rot -8.9 -90 -0.0703>}]
}
constraints [{bone1Name=dyn_liquid, bone2Name=dyn_pack, swing1LimitDegrees=20, swing2LimitDegrees=20, twistLimitDegrees=0, twistMotion=Locked}]
