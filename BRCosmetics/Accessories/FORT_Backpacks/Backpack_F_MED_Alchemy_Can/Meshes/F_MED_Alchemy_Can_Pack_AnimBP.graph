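# Animation Blueprint for the Alchemy Can backpack cosmetic, which fakes sloshing liquid with secondary motion.
# A RigidBody node simulates the pack's physics asset in root-bone space, then two additive Transform Bone nodes scale dyn_liquid on Z.
# Each update reads dyn_liquid's pitch delta from ref pose relative to dyn_pack.
# Positive pitch blends in the stretch (Liquid Scale), negative the squash (Liquid Trans, a scale despite its name).
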AnimBlueprint /BRCosmetics/Accessories/FORT_Backpacks/Backpack_F_MED_Alchemy_Can/Meshes/F_MED_Alchemy_Can_Pack_AnimBP
target_skeleton=&/F_MED_Alchemy_Can_Pack_Skeleton
graphs [
 AnimGraph {nodes [
   n1 = RigidBody(ComponentPose=Local To Component(LocalPose=Local Space Ref Pose())) {
    bClampLinearTranslationLimitToRefPose=true
    bOverrideWorldGravity=true
    overridePhysicsAsset=&/F_MED_Alchemy_Can_Pack_Physics
    simulationSpace=BaseBoneSpace
    alphaBoolBlend {blendOption=Linear}
    baseBoneRef {boneName=root}
    componentAppliedLinearAccClamp=<3000 3000 3000>
    componentLinearAccScale=<0.334 0.334 0.334>
    componentLinearVelScale=<0.334 0.334 0.334>
    overrideWorldGravity {z=-980}
   }
   n2 = "Transform (Modify) Bone - Bone: dyn_liquid"(ComponentPose=@n1, Alpha=Get Liquid Trans() -> (Liquid Trans:double), Translation="0,0,0", Rotation="0,0,0", Scale="1,1,0.85") {scaleMode=BMM_Additive, alphaBoolBlend {blendOption=Linear}, boneToModify {boneName=dyn_liquid}, scale {z=0.85}}
   n3 = Get Liquid Scale() -> (Liquid Scale:double)
   Output Pose(Result=Component To Local(ComponentPose="Transform (Modify) Bone - Bone: dyn_liquid"(ComponentPose=@n2, Alpha=@n3, Translation="0,0,0", Rotation="0,0,0", Scale="1,1,1.35") {scaleMode=BMM_Additive, alphaBoolBlend {blendOption=Linear}, boneToModify {boneName=dyn_liquid}, scale {z=1.35}})) {layerGroup=None}
  ]}
 EventGraph {nodes [
   Event BlueprintUpdateAnimation()
   > n1 = Sequence!(execute) -> (then_0, then_1, then_2:Exec)
   n2 = GetDeltaTransformFromRefPose(self=GetOwningComponent(), BoneName=dyn_liquid, BaseName=dyn_pack)
   > "Set Y (Pitch)"("Y (Pitch)":double=BreakRotator(InRot=BreakTransform(InTransform=@n2).Rotation).Pitch) -> (Output_Get:double)
   Set Liquid Scale(execute=@n1.then_1, Liquid Scale:double=MapRangeClamped(Value="Get Y (Pitch)"() -> ("Y (Pitch)":double), InRangeA=0, InRangeB=15)) -> (Output_Get:double)
   Set Liquid Trans(execute=@n1.then_2, Liquid Trans:double=MapRangeClamped(Value="Get Y (Pitch)"(), InRangeA=0, InRangeB=-15)) -> (Output_Get:double)
   TryGetPawnOwner()
   "Absolute (Float)"(A=0)
   Branch(Condition="float > float"(A=0, B=0) {is=title K2Node_CallFunction})
   Set Liquid Scale(Liquid Scale=0)
  ]}
]
variables [
 Liquid Scale {type=number}
 Liquid Trans {type=number}
 "Y (Pitch)" {type=number}
]
