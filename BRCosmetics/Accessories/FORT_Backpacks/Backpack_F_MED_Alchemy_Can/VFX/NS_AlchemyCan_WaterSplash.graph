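# Niagara water splash effect for the Alchemy Can backpack cosmetic, tinted by the User.LiquidCOlor parameter.
# Splash bursts sprite droplets shortly after start, launched upward in a cone from a small offset sphere and pulled down by local gravity.
# Splash001 bursts fewer, larger sprites from a flat ring, drifting gently upward without gravity.
# Both use capped inherited velocity, grow over life and animate dynamic material parameters with curves.
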NiagaraSystem /BRCosmetics/Accessories/FORT_Backpacks/Backpack_F_MED_Alchemy_Can/VFX/NS_AlchemyCan_WaterSplash
dependencies {usedDataInterfaces [&NiagaraDataInterfaceCurve], usedDynamicInputs [], usedModules [&SystemState]}
emitters {
 Splash {
  emitterUpdateScript [
   &EmitterState()
   &SpawnBurst_Instantaneous(Spawn Count=13, Spawn Time=0.05, Use Spawn Probability=false)
  ]
  particleSpawnScript [
   &InitializeParticle(
    Color=$User.LiquidCOlor
    Lifetime=&RandomRangeFloat(Minimum=0.45, Maximum=0.6)
    Sprite Rotation Mode=::NewEnumerator2
    Sprite Size Mode=::NewEnumerator3
    Uniform Sprite Size=&RandomRangeFloat(Minimum=4, Maximum=6.5)
    UsePositionOffset=false
    Write Material Random=true
   )
   &AddVelocityInCone(
    Cone Angle=30
    Cone Axis=<0 0 1>
    Module.Use Velocity Falloff On Cone Axis=false
    Velocity Strength=&RandomRangeFloat(Minimum=28, Maximum=45)
   )
   &SphereLocation(
    MaskSpawnGroup=false
    Offset=<3 -4 -7>
    Override Seed=false
    Sphere Radius=2
    Surface Only=false
    UseNonUniformScale=false
    UseOffset=true
    UseSphereOrientation=false
   )
   &CameraOffset(Camera Offset Amount=&RandomRangeFloat(Minimum=0, Maximum=1))
  ]
  particleUpdateScript [
   &DynamicMaterialParameters(
    Index 0 Param 1=&FloatFromCurve(FloatCurve=DataInterface() {propertyValues {Curve {keys [<at 0 0 RCIM_Cubic>, {arriveTangent=3.8258, interpMode=RCIM_Cubic, leaveTangent=3.8258, tangentMode=RCTM_User, time=1, value=2}]}}})
    Index 0 Param 2=&FloatFromCurve(FloatCurve=DataInterface() {propertyValues {Curve {keys [<at 0 0 RCIM_Cubic>, {arriveTangent=2.9697, interpMode=RCIM_Cubic, leaveTangent=2.9697, tangentMode=RCTM_User, time=1, value=1.5}]}}})
    Index 0 Param 3=0.1
    Index 0 Param 4=1
    Module.Param0WriteEnabled=true
    Module.Param1WriteEnabled=true
    Module.Param2WriteEnabled=true
    Module.Param3WriteEnabled=true
   )
   &InheritVelocity(Module.Limit Speed=true, Velocity Limit=0.5)
   &GravityForce(Coordinate Space=::Local, Gravity=<0 0 -130>)
   &SolveForcesAndVelocity(Clamp Velocity=false, Limit Acceleration=false)
   &ParticleState()
   &ScaleSpriteSize(Scale Factor=&Vector2DFromFloat(Value=&FloatFromCurve(FloatCurve=DataInterface() {propertyValues {Curve {keys [<at 0 0>,<at 1 1>]}}})))
  ]
  rendererClasses [&NiagaraSpriteRendererProperties]
  renderers [{rendererClass=&NiagaraSpriteRendererProperties}]
 }
 Splash001 {
  emitterUpdateScript [
   &EmitterState()
   &SpawnBurst_Instantaneous(Spawn Count=6, Use Spawn Probability=false)
  ]
  particleSpawnScript [
   &InitializeParticle(
    Color=$User.LiquidCOlor
    Lifetime=&RandomRangeFloat(Minimum=0.35, Maximum=0.5)
    Sprite Rotation Mode=::NewEnumerator2
    Sprite Size=&RandomRangeVector2D(Minimum=<6 6>, Maximum=<9 9>)
    Sprite Size Mode=::NewEnumerator4
    UsePositionOffset=false
    Write Material Random=true
   )
   &CylinderLocation(
    Cylinder Height=0
    Cylinder Radius=1
    MaskSpawnGroup=false
    Module.Enable Lathe Profile=false
    Module.Override Randomness=false
    Module.Surface Only=true
    Offset=<5 -5 -2.5>
    Surface Only Band Thickness=0
   )
   &AddVelocity(Velocity=&RandomRangeVector(Minimum=<-0.5 -0.5 0.8>, Maximum=<0.5 0.5 1>))
  ]
  particleUpdateScript [
   &DynamicMaterialParameters(
    Index 0 Param 1=&FloatFromCurve(FloatCurve=DataInterface() {propertyValues {Curve {keys [<at 0 0 RCIM_Cubic>, {arriveTangent=3.8258, interpMode=RCIM_Cubic, leaveTangent=3.8258, tangentMode=RCTM_User, time=1, value=2}]}}})
    Index 0 Param 2=&FloatFromCurve(FloatCurve=DataInterface() {propertyValues {Curve {keys [<at 0 0 RCIM_Cubic>, {arriveTangent=2.9697, interpMode=RCIM_Cubic, leaveTangent=2.9697, tangentMode=RCTM_User, time=1, value=1.5}]}}})
    Index 0 Param 3=0.1
    Index 0 Param 4=1
    Module.Param0WriteEnabled=true
    Module.Param1WriteEnabled=true
    Module.Param2WriteEnabled=true
    Module.Param3WriteEnabled=true
   )
   &InheritVelocity(Module.Limit Speed=true, Velocity Limit=0.15)
   &SolveForcesAndVelocity(Clamp Velocity=false, Limit Acceleration=false)
   &ParticleState()
   &ScaleSpriteSize(Scale Factor=&Vector2DFromFloat(Value=&FloatFromCurve(FloatCurve=DataInterface() {propertyValues {Curve {keys [<at 0 0>,<at 0.5 1>]}}})))
  ]
  rendererClasses [&NiagaraSpriteRendererProperties]
  renderers [{rendererClass=&NiagaraSpriteRendererProperties}]
 }
}
summary {
 systemName=NS_AlchemyCan_WaterSplash
 emitters [
  {emitterName=Splash, rendererClasses [&NiagaraSpriteRendererProperties]}
  {emitterName=Splash001, rendererClasses [&NiagaraSpriteRendererProperties]}
 ]
 userVariables [User.LiquidCOlor {type=LinearColor, description="", defaultValue {struct=LinearColor, value=<rgba 0.2789 0.402 0.063 1>}}]
}
